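; Gameplay cue for the WildEstate Zeropoint Dash (IronLog variant) that drives dash visual effects, distinguishing the local player from others.
; OnActive spawns a River Styx dash Niagara system on the pelvis and calls the parent only when the target's instigator is locally controlled.
; The Niagara burst handler sets SpeedLineCount on a spawned component, zeroing it for non-local pawns.
; OnActive and OnExecute contain debug prints, and OnExecute defers to the parent.
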(fn OnActive
  (params (MyTarget :type Actor%20Object%20Reference) (Parameters :type Gameplay%20Cue%20Parameters%20Structure%20%28by%20ref%29))
  (bind (_normalizedmagnitude _rawmagnitude _effectcontext _matchedtagname _originaltag _aggregatedsourcetags _aggregatedtargettags _location _normal _instigator _effectcauser _sourceobject _physicalmaterial _gameplayeffectlevel _abilitylevel _targetattachcomponent) (Ability|GameplayCue|BreakGameplayCueParameters Parameters))
  (if (Pawn|IsLocallyControlled (Variables|/Script/Engine.Actor|GetInstigator MyTarget))
    (Development|PrintString 0 "IS LOCAL" true true "(R=0.000000,G=1.000000,B=0.141573,A=1.000000)")
    (Niagara|SpawnSystemAttached "/SunRoseEnvGameplay/Assets/FX/NS_RiverStyx_Dash_Dashing.NS_RiverStyx_Dash_Dashing" _targetattachcomponent "pelvis" "0, 0, 0" "0, 0, 0" "SnapToTarget")
    (return (|CallParentFunction MyTarget Parameters))
    (else
      (Development|PrintString 0 "NOT LOCAL" true true "(R=1.000000,G=0.049180,B=0.000000,A=1.000000)"))))

(fn OnBurst(Niagara)
  (params (MyTarget :type Actor%20Object%20Reference) (Parameters :type Gameplay%20Cue%20Parameters%20Structure%20%28by%20ref%29) (ParticleComponents :type Array%20of%20Niagara%20Particle%20System%20Component%20Object%20References) (AudioComponents :type Array%20of%20Audio%20Component%20Object%20References) (BurstCameraShakeInstance :type Camera%20Shake%20Base%20Object%20Reference) (BurstDecalInstance :type Decal%20Actor%20Object%20Reference))
  (bind _output (Utilities|Array|Get(acopy) ParticleComponents))
  (bind _asfort_player_pawn (Utilities|Casting|CastToFortPlayerPawn MyTarget))
  (if (Pawn|IsLocallyControlled _asfort_player_pawn)
    (Effects|Components|ParticleSystem|SetIntParameter _output "SpeedLineCount" (LocalVariables|GetSpeedLineCount))
    (else
      (Effects|Components|ParticleSystem|SetIntParameter _output "SpeedLineCount" 0)))
  (return))

(fn OnExecute
  (params (MyTarget :type Actor%20Object%20Reference) (Parameters :type Gameplay%20Cue%20Parameters%20Structure%20%28by%20ref%29))
  (Development|PrintString 0)
  (return (|CallParentFunction MyTarget Parameters)))
